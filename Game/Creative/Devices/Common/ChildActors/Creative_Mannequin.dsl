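; Creative device child-actor mannequin that wears character, back bling and shoe cosmetics and plays an idle animation.
; It either takes explicit items and variant tags or mirrors a tracked pawn's loadout, rebinding to that pawn's customization-finished event and reapplying only on change, with fixups for problem skins.
; Idle montages or sequences load asynchronously, positioned from a server start time.
; After reinitializing, it broadcasts meshes and overridden materials as memory costs.
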(event Custom|Update%20Cosmetic%20and%20Play%20Idle
  (params (Pawn :type Fort%20Player%20Pawn%20Object%20Reference))
  ;[1 0.5253 0 0.2] Update Cosmetic from a pawn. Checks that the pawn isn't currently loading a customization, and if it is, we wait and try again once it has. This fixes two problems: a) If a tracked pawn changes customization during the game, the mannequin will now update, and b) potential fix to an async load if we try and get the cosmetic loadout of a pawn that hasn't finished loading yet.
  ;;[1 1 1] fixed a rare bug where the event has fired but IsCharacterCustomizationLoadingCompleted is still false, normally it changes to true after this event is called.
  (Utilities|WaitforCharacterCustomizationLoadingCompleted Pawn
    (:Completed
      (CallFunction|UpdateCosmeticfromPawn Pawn)
      (CallFunction|SetAnimationfromLoadedAnim))
    (:NotCompleted
      (CallFunction|UpdateCosmeticAndPlayIdle Pawn))))

(event Custom|Update%20Cosmetic%20From%20Pawn
  (params (Pawn :type Fort%20Player%20Pawn%20Object%20Reference))
  ;[1 0.5253 0 0.2] Update Cosmetic from a pawn. Checks that the pawn isn't currently loading a customization, and if it is, we wait and try again once it has. This fixes two problems: a) If a tracked pawn changes customization during the game, the mannequin will now update, and b) potential fix to an async load if we try and get the cosmetic loadout of a pawn that hasn't finished loading yet.
  (bind _current_tracked_pawn (Variables|Default|GetCurrent%20Tracked%20Pawn)
    (:then
      (Customization|UnbindEventfromOnFinishedCharacterCustomization _current_tracked_pawn (delegate Custom|Update%20Cosmetic%20and%20Play%20Idle)))
    (:else
      _))
  (bind _output_get (Variables|Default|SetCurrent%20Tracked%20Pawn Pawn))
  (Utilities|IsValid _output_get
    (:Is%20Valid
      (Customization|BindEventtoOnFinishedCharacterCustomization Pawn (delegate Custom|Update%20Cosmetic%20and%20Play%20Idle))
      (if (Customization|IsCharacterCustomizationLoadingCompleted Pawn)
        (CallFunction|UpdatefromLoadout Pawn)))))

(event Custom|Update%20Cosmetic
  (params (Character :type Athena%20Character%20Item%20Definition%20Object%20Reference) (CharacterVariantTags :type Gameplay%20Tag%20Container%20Structure) (BackBling :type Athena%20Backpack%20Item%20Definition%20Object%20Reference) (BackBlingVariantTags :type Gameplay%20Tag%20Container%20Structure))
  ;[0.7448 0.585 0.1685] Update the backpack and character cosmetics. Note: Update backpack and character together to avoid sync loads and ensure the correct variants are applied
  (Customization|SetAthenaBackpackandCharacterDefandVariants BackBling (CallFunction|CreateVariantsfromTagContainer BackBling BackBlingVariantTags) Character (CallFunction|CreateVariantsfromTagContainer Character CharacterVariantTags))
  ;;[1 1 1] FIX FORT-648331 / Use Female Skeletal Mesh the associated HID is set as Female character
  (if (ScalableFloat|GetValueasBool (Variables|Default|GetHF%20Enable%20Female%20Skeleton))
    (Utilities|IsValid Character
      (:Is%20Valid
        (Components|SkeletalMesh|SetSkeletalMeshAsset (Variables|SkeletalMeshActor|GetSkeletalMeshComponent) (Utilities|ResolveSoftReference (Utilities|Select :Index (Class|AthenaCharacterItemDefinition|GetCharacterGenderNoLoadout Character) :Invalid (Variables|SkeletalMesh|GetSkeletal%20Mesh%20Male) :Male (Variables|SkeletalMesh|GetSkeletal%20Mesh%20Male) :Female (Variables|SkeletalMesh|GetSkeletal%20Mesh%20Female) :Both (Variables|SkeletalMesh|GetSkeletal%20Mesh%20Male))))))))

(event Custom|Load%20and%20Play%20Idle%20Montage
  (params (Animation :type Soft%20Object%20Reference) (AnimStartWorldTime :type Float%20%28double-precision%29))
  ;[1 0.8528 0] Loads and caches the intended montage then if we're not currently loading, applies it (if we're currently loading, do nothing as it will be applied after loading.)
  (Variables|Animation|SetAnim%20Start%20World%20Time AnimStartWorldTime)
  (bind _loaded_object (Creative|Loading|AsyncLoadEditorSafe Animation)
    (:Loaded
      (bind _asanim_montage (Utilities|Casting|CastToAnimMontage _loaded_object)
        (:then
          (Variables|Default|SetIdle%20Montage _asanim_montage)
          (Variables|Default|SetIdle%20Pose 0))
        (:CastFailed
          (bind _asanim_sequence (Utilities|Casting|CastToAnimSequence _loaded_object))
          (Variables|Default|SetIdle%20Pose _asanim_sequence)
          (Variables|Default|SetIdle%20Montage 0)))
      (CallFunction|SetAnimationfromLoadedAnim))))

(event Custom|Set%20Animation%20From%20Loaded%20Anim
  (bind _randomidlestart (CallFunction|GetRandomIdleStart))
  (bind _anim_start_world_time (CallFunction|GetAnimationStartTime))
  (bind _skeletalmeshcomponent (Variables|SkeletalMeshActor|GetSkeletalMeshComponent)
    (:then
      (bind _idle_montage (Variables|Default|GetIdle%20Montage)
        (:then
          (Components|Animation|SetAnimation _skeletalmeshcomponent _idle_montage)
          (CallFunction|SetSequencePosition _idle_montage _randomidlestart _anim_start_world_time))
        (:else
          (bind _idle_pose (Variables|Default|GetIdle%20Pose)
            (:then
              (Components|Animation|SetAnimation _skeletalmeshcomponent _idle_pose)
              (CallFunction|SetSequencePosition _idle_pose _randomidlestart _anim_start_world_time))
            (:else)))))
    (:else)))

(event EventHandleReInitializeComplete
  ;[1 0.7554 0.3604] Pass through to Set Animation From Loaded Idle, possibly triggering the device to load our animation if it was not ready before.
  (Default|CallOnFinishedLoadingCharacter)
  (CallFunction|SetAnimationfromLoadedAnim)
  (CallFunction|UpdateMemoryCosts))

(event Custom|Load%20Generic%20Character
  (bind _loaded_object (Creative|Loading|AsyncLoadEditorSafe (Variables|Default|GetDefault%20Character))
    (:Loaded
      (bind _asathena_character_item_definition (Utilities|Casting|CastToAthenaCharacterItemDefinition _loaded_object))
      (Customization|SetAthenaCharacterDefandVariants _asathena_character_item_definition (Utilities|Array|MakeArray)))))

(fn CreateVariantsFromTagContainer
  (params (ItemDefinition :type Athena%20Cosmetic%20Item%20Definition%20Object%20Reference) (VariantTags :type Gameplay%20Tag%20Container%20Structure))
  (Utilities|IsValid ItemDefinition
    (:Is%20Valid
      (if (Utilities|GameplayTagContainerNotEmpty VariantTags)
        (bind _returnvalue (Items|GetAllFortVariantTokens))
        (bind _returnvalue_2 (Cosmetic|MakeVariantChannelInfosfromVariantTagContainer ItemDefinition VariantTags _returnvalue))
        (return _returnvalue_2)
        (else
          _)))
    (:Is%20Not%20Valid
      _))
  (return (Utilities|Array|MakeArray)))

(fn GetAnimationStartTime
  ;[1 0.586 0 0.2] If we have an animation start world time set, that's when the animation should have started (synching to clients out of relevancy when it started). If the time is close to 0 difference from current, use 0 instead to stop players missing the start while in range due to latency.
  (if (== (Variables|Animation|GetAnim%20Start%20World%20Time) 0)
    (return 0.0)
    (else
      (return (Math|Float|SelectFloat 0.0 (- (Fortnite|Sync|Time|GetServerWorldTimeSeconds) (Variables|Animation|GetAnim%20Start%20World%20Time)) (< (- (Fortnite|Sync|Time|GetServerWorldTimeSeconds) (Variables|Animation|GetAnim%20Start%20World%20Time)) (Variables|Animation|GetAnimation%20Start%20Tolerance)))))))

(fn GetInitialIdleAnimationPosition
  (return 0.0))

(fn GetRandomIdleStart
  (return false))

(fn HandleSpecialCases
  (params (ActorCosmeticContext :type Actor%20Object%20Reference))
  ;[1 1 1] Some characters don't interact well with the mannequin, and need fixup to work properly. In this function we attempt to fix up each mannequin
  (HandleSpecialCases|SpecialCaseShapeShifter (CosmeticLoadout|GetItemDefinitionTypefromLoadoutwithActor ActorCosmeticContext "(TagName=\"LoadoutSlot.Character\")" "/Script/FortniteGame.AthenaCharacterItemDefinition")))

(fn SetSequencePosition
  (params (Sequence :type Anim%20Sequence%20Base%20Object%20Reference) (RandomIdleStart :type Boolean) (InitialIdleAnimPosition :type Float%20%28double-precision%29))
  (bind _skeletalmeshcomponent (Variables|SkeletalMeshActor|GetSkeletalMeshComponent)
    (:then
      (Components|Animation|SetPosition _skeletalmeshcomponent (Math|Float|SelectFloat (Math|Random|RandomFloatinRange 0.0 (Animation|GetPlayLength Sequence)) (Math|Float|%25%28Float%29 InitialIdleAnimPosition (Animation|GetPlayLength Sequence)) RandomIdleStart) true))
    (:else)))

(fn SpecialCase_ShapeShifter
  (params (CID :type Athena%20Character%20Item%20Definition%20Object%20Reference))
  (if (Utilities|Equal(SoftObjectReference) (Utilities|ResolveSoftReference CID) (LocalVariables|GetItem%20Def%20To%20Test))
    (Customization|RemovePart self "MiscOrTail")))

(fn UpdateFromLoadout
  (params (CosmeticActorContext :type Actor%20Object%20Reference))
  ; Cache our current Loadout's Item Definitions and Variant Channels
  (Utilities|IsValid CosmeticActorContext
    (:Is%20Valid
      (LocalVariables|SetIn%20Character%20Item%20Def (CosmeticLoadout|GetItemDefinitionTypefromLoadoutwithActor CosmeticActorContext "(TagName=\"LoadoutSlot.Character\")" "/Script/FortniteGame.AthenaCharacterItemDefinition"))
      (LocalVariables|SetIn%20Variant%20Channels (CosmeticLoadout|GetAllLoadoutVariantInfofromActor CosmeticActorContext))
      (LocalVariables|SetIn%20Shoes%20Item%20Def (CosmeticLoadout|GetItemDefinitionfromLoadoutwithActor CosmeticActorContext "(TagName=\"LoadoutSlot.ExtraPart\")"))
      ;[1 0.5253 0 0.2] Updates our Mannequin from the Loadout if anything has changed, then saves a copy to match against the next update.
      (if (Math|Boolean|ORBoolean (Utilities|NotEqual(Object) (LocalVariables|GetIn%20Character%20Item%20Def) (Variables|Default|GetCurrent%20Tracked%20Character%20Item%20Def)) (not (Utilities|Array|Identical (LocalVariables|GetIn%20Variant%20Channels) (Variables|Default|GetCurrent%20Tracked%20Variant%20Channels))) (!= (LocalVariables|GetIn%20Shoes%20Item%20Def) (Variables|Default|GetCurrent%20Tracked%20Shoes%20Item%20Def)))
        (Customization|SetAthenaCharacterDefandVariants (LocalVariables|GetIn%20Character%20Item%20Def) (LocalVariables|GetIn%20Variant%20Channels))
        (Variables|Default|SetCurrent%20Tracked%20Character%20Item%20Def (LocalVariables|GetIn%20Character%20Item%20Def))
        (Variables|Default|SetCurrent%20Tracked%20Variant%20Channels (LocalVariables|GetIn%20Variant%20Channels))
        (Variables|Default|SetCurrent%20Tracked%20Shoes%20Item%20Def (LocalVariables|GetIn%20Shoes%20Item%20Def))
        (CallFunction|HandleSpecialCases (Variables|GetCosmeticActorContext)) ; Some skins break the mannequin in interesting ways. This function will special case them. 
        ;[1 1 1] Support Shoe cosmetics
        (bind _asathena_character_part_item_definition (Utilities|Casting|CastToAthenaCharacterPartItemDefinition (LocalVariables|GetIn%20Shoes%20Item%20Def)))
        (if (Utilities|Array|IsNotEmpty (Cosmetics|GetCharacterParts _asathena_character_part_item_definition))
          (Utilities|IsValid (Utilities|Array|Get(acopy) (Cosmetics|GetCharacterParts _asathena_character_part_item_definition))
            (:Is%20Valid
              (Customization|SetPart "ExtraPart" (Utilities|Array|Get(acopy) (Cosmetics|GetCharacterParts _asathena_character_part_item_definition))))))))))

(fn UpdateAnimationTickRate
  (params (Enable%20Update%20Rate%20Optimizations :type Boolean))
  (bind _skeletalmeshcomponent (Variables|SkeletalMeshActor|GetSkeletalMeshComponent))
  (Variables|/Script/Engine.SkinnedMeshComponent|SetVisibilityBasedAnimTickOption (select Enable%20Update%20Rate%20Optimizations "OnlyTickPoseWhenRendered" "AlwaysTickPose") _skeletalmeshcomponent)
  (Variables|/Script/Engine.SkinnedMeshComponent|SetbEnableUpdateRateOptimizations Enable%20Update%20Rate%20Optimizations _skeletalmeshcomponent))

(fn UpdateMemoryCosts
  (Variables|Transient|SetMemoryObjects (Utilities|Array|MakeArray (Cosmetics|GetAthenaCharacterItemDefinition)))
  (bind _skeletalmeshcomponents (Customization|GetAllPartSkeletalMeshComponents))
  (for _array_element _skeletalmeshcomponents
    (Utilities|IsValid _array_element
      (:Is%20Valid
        (bind _skeletalmesh (Variables|/Script/Engine.SkinnedMeshComponent|GetSkeletalMesh _array_element)
          (:then
            (LocalVariables|SetLocal%20Materials (Rendering|Material|GetMaterials _array_element))
            (Utilities|Array|Add (Variables|Transient|GetMemoryObjects) _skeletalmesh)
            ;[1 0.586 0 0.2] If the material isn't the material that should be on the mesh, add it.
            (for (_array_element_2 _array_index) (LocalVariables|GetLocal%20Materials)
              (LocalVariables|SetMaterialFound _array_element_2)
              (bind _materialfound (LocalVariables|GetMaterialFound)
                (:then
                  (bind _asmaterial_instance_dynamic (Utilities|Casting|CastToMaterialInstanceDynamic _materialfound)
                    (:then
                      (bind _output_get (LocalVariables|SetMaterialFound (Variables|/Script/Engine.MaterialInstance|GetParent _asmaterial_instance_dynamic))))
                    (:CastFailed
                      _))
                  (bind _output_materialinterface (Utilities|Struct|BreakSkeletalMaterial (Utilities|Array|Get(aref) (Variables|/Script/Engine.SkeletalMesh|GetMaterials _skeletalmesh) _array_index)))
                  (if (!= _output_get _output_materialinterface)
                    (Utilities|Array|AddUnique (Variables|Transient|GetMemoryObjects) (LocalVariables|GetMaterialFound))))
                (:else))))
          (:else)))))
  (Default|CallOnMemoryCostUpdated (Variables|Transient|GetMemoryObjects)))

(fn ConstructionScript)
